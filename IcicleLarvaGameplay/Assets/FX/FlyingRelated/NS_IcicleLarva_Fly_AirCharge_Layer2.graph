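# Niagara system for the second layer of the Icicle Larva's flying air-charge effect; its only emitter, Light, is disabled, so it currently shows nothing.
# The emitter burst-spawns warm-tinted particles with sprite and light renderers that persist past their lifetime.
# Each update repositions them at a fixed offset along the owner's normalized velocity and randomizes light intensity and color brightness for flicker.
# A uDuration user parameter is exposed but not referenced by the emitter.
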NiagaraSystem /IcicleLarvaGameplay/Assets/FX/FlyingRelated/NS_IcicleLarva_Fly_AirCharge_Layer2
dependencies {usedDataInterfaces [], usedDynamicInputs [], usedModules [&SystemState]}
emitters {Light {
  bEnabled=false
  emitterUpdateScript [
   &EmitterState(Inactive Response=::NewEnumerator1, Life Cycle Mode=::NewEnumerator1, Loop Duration=1, UseLoopDelay=false)
   &SpawnBurst_Instantaneous(Use Spawn Probability=false, UseLoopCountLimit=false)
  ]
  particleSpawnScript [&InitializeParticle(
    Color=<rgba 1 0.7339 0.6094 1>
    Lifetime=0.5
    Position Offset=&TransformVector(Vector=<0 0 0>, Destination Space=::World)
    Position Offset Coordinate Space=::Local
    UsePositionOffset=true
    Write Material Random=true
   )]
  particleUpdateScript [
   &ParticleState(Kill Particles When Lifetime Has Elapsed=false)
   /SetVariables(
    Particles.LightExponent=4
    Particles.LightRadius=1000
    Particles.PointLight Intensity:NiagaraFloat=&Multiply_Float(A=1, B=&RandomRangeFloat(Minimum=0, Maximum=1, Evaluation Type=::NewEnumerator1))
   ) {isSetParameters=true}
   /SetVariables(Particles.Position=&AddVectorToPosition(
     Position=$Engine.Emitter.SimulationPosition
     Vector=&Multiply_VectorByFloat(
      Vector=&NormalizeVector(Vector To Normalize=$Engine.Owner.Velocity, Safe Normalize=true, Safe Fallback Threshold=0, Safe Fallback Vector=<0 0 0.001>)
      Float=150
     )
    )) {isSetParameters=true}
   &ScaleColor(
    Scale RGB=&VectorFromFloat(Value=&RandomRangeFloat(Minimum=5, Maximum=10, Evaluation Type=::NewEnumerator1))
    ScaleA=true
    ScaleRGB=true
   )
  ]
  rendererClasses [&NiagaraSpriteRendererProperties, &NiagaraLightRendererProperties]
  renderers [
   {rendererClass=&NiagaraSpriteRendererProperties}
   {rendererClass=&NiagaraLightRendererProperties, rendererIndex=1}
  ]
 }}
summary {
 systemName=NS_IcicleLarva_Fly_AirCharge_Layer2
 emitters [{
   bEnabled=false
   emitterName=Light
   rendererClasses [&NiagaraSpriteRendererProperties, &NiagaraLightRendererProperties]
  }]
 userVariables [User.uDuration {type=NiagaraFloat, defaultValue=1.5, description=""}]
}
